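# Translucent UI material for Donut challenge icons, with adjustable tint and a vertical erase cutoff.
# EraseLine builds a hard mask from the UV's V coordinate that blanks the Icon texture below that line.
# The IsNoiseText switch uses R/G/B-weighted channels as both color base and mask; otherwise it keeps RGB and derives the mask from boosted channel sum.
# Tint blends the color toward TintColor, while the TintLine parameter is unconnected.
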Material /Game/Athena/UI/Challenges/Art/Donut/M_UI_DonutIcons
graph {
 properties {
  bUsedWithStaticMesh=false
  blendMode=BLEND_Translucent
  materialDomain=MD_UI
  pixelDepthOffsetMode=PDOM_Legacy
  usedShadingModels=Default Lit
 }
 nodes [
  ScalarParameter() {parameterName=TintLine}
  n1 = TextureSampleParameter2D() {parameterName=Icon, texture=&/T_UI_Donut_PC_Icon}
  n2 = MaterialFunctionCall(Float2=TextureCoordinate()) {materialFunction=&BreakOutFloat2Components}
  n3 = ScalarParameter() {defaultValue=1, parameterName=EraseLine}
  n4 = OneMinus(None=Ceil(None=Subtract(A=@n2.G, B=@n3) {constB=0.8}))
  n5 = Multiply(A=@n1.R, B=@n4)
  n6 = Multiply(A=ScalarParameter() {defaultValue=1, parameterName=R}, B=@n5)
  n7 = Multiply(A=@n1.G, B=@n4)
  n8 = Multiply(A=ScalarParameter() {parameterName=G}, B=@n7)
  n9 = ScalarParameter() {parameterName=B}
  n10 = Multiply(A=@n1.B, B=@n4)
  n11 = Clamp(None=Add(A=Add(A=@n6, B=@n8), B=Multiply(A=@n9, B=@n10)))
  n12 = VectorParameter() {parameterName=TintColor, defaultValue=<rgba 1 1 1 1>}
  StaticSwitchParameter(
   True=LinearInterpolate(A=@n11, B=@n12.RGB, Alpha=ScalarParameter() {parameterName=Tint})
   False=LinearInterpolate(A=Multiply(A=@n1.RGB, B=@n4), B=@n12.RGB, Alpha=ScalarParameter() {parameterName=Tint})
  ) {parameterName=IsNoiseText}
  n13 = Add(A=@n10, B=Add(A=@n5, B=@n7))
  n14 = ScalarParameter() {defaultValue=20, parameterName=ShapeMult}
  StaticSwitchParameter(True=@n11, False=Clamp(None=Multiply(A=@n13, B=@n14) {constB=20})) {parameterName=IsNoiseText}
 ]
}
